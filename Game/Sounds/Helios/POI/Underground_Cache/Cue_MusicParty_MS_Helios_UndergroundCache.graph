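# Sound cue for looping diegetic party music at the Helios Underground Cache point of interest, set up with speaker attenuation.
# A looping wave player plays the underground diegetic loop through an enveloper whose volume curve rises from silence to full at the start.
# The cue sits in the POI music sound class, caps concurrent instances, runs a small Bluetooth speaker effect chain and sends to a music audio-analysis submix.
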SoundCue /Game/Sounds/Helios/POI/Underground_Cache/Cue_MusicParty_MS_Helios_UndergroundCache
root=Enveloper_0
nodes [
 {
  id=Enveloper_0
  children [WavePlayer_1]
  properties {pitchCurve {editorCurveData {keys [{time=0}]}}, volumeCurve {editorCurveData {keys [<at 0 0>,<at 1 1>]}}}
 }
 {id=WavePlayer_1, properties {bLooping=true, soundWaveAssetPtr=/Game/Sounds/Helios/POI/ForestLodge/s28_Underground_Diegetic_Loop}}
]
properties {
 attenuationSettings=&/Helios_Music_UndergroundCache_Speaker_Att
 bOverrideConcurrency=true
 maxDistance=1250
 soundClassObject=/Game/Sounds/FortSoundClassesAndModes/Fort_Music_POI_PSM
 sourceEffectChain=/Game/Sounds/Artemis/Biomes/HollowBaseParty/Bluetooth/Heist_Bluetooth_SmallSpeaker_SrcFXChain
 volumeMultiplier=0.25
 concurrencyOverrides {maxCount=4}
 soundSubmixSends [{minSendLevel=0, soundSubmix=&Fortnite_AudioAnalysis_Music}]
}
